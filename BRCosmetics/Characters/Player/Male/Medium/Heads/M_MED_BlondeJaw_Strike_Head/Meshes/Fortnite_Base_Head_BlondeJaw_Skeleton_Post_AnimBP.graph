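# Post-process AnimBP for the BlondeJaw Strike male head skeleton, adding a Control Rig pass and a front-end backpack pose adjustment.
# The input pose runs through the head's post Control Rig, then additive Modify Bone nodes rotate lower spine, head and neck in component space (spine_04/05 offsets are zero).
# Their shared alpha is Frontend_Backpack_Pose, read from the same-named animation curve on update, gated by bIsInFrontEnd.
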AnimBlueprint /BRCosmetics/Characters/Player/Male/Medium/Heads/M_MED_BlondeJaw_Strike_Head/Meshes/Fortnite_Base_Head_BlondeJaw_Skeleton_Post_AnimBP
target_skeleton=/BRCosmetics/Characters/Player/Common/Fortnite_Base_Head/BlondeJaw/Fortnite_Base_Head_BlondeJaw_Skeleton
graphs [
 AnimGraph {nodes [
   n1 = Control Rig(Source=Input Pose()) {
    alphaBoolBlend {blendOption=Linear}
    controlRigAssetReference {blueprintRigClass=&/M_MED_BlondeJaw_Strike_Head_CtrlRig_Post.M_MED_BlondeJaw_Strike_Head_CtrlRig_Post_C}
   }
   n2 = Local To Component(LocalPose=@n1)
   n3 = Get Frontend_Backpack_Pose() -> (Frontend_Backpack_Pose:float)
   n4 = "Transform (Modify) Bone - Bone: spine_03"(
    ComponentPose="Transform (Modify) Bone - Bone: spine_04"(
     ComponentPose="Transform (Modify) Bone - Bone: spine_05"(ComponentPose=@n2, Alpha=@n3, Translation="0,0,0", Rotation="0,0,0", Scale="1,1,1") {rotationMode=BMM_Additive, alphaBoolBlend {blendOption=Linear}, boneToModify {boneName=spine_05}}
     Alpha=@n3
     Translation="0,0,0"
     Rotation="0,0,0"
     Scale="1,1,1"
    ) {rotationMode=BMM_Additive, alphaBoolBlend {blendOption=Linear}, boneToModify {boneName=spine_04}}
    Alpha=@n3
    Translation="0,0,0"
    Rotation="0,0,10"
    Scale="1,1,1"
   ) {rotationMode=BMM_Additive, alphaBoolBlend {blendOption=Linear}, boneToModify {boneName=spine_03}, rotation {roll=10}}
   n5 = "Transform (Modify) Bone - Bone: spine_02"(ComponentPose=@n4, Alpha=@n3, Translation="0,0,0", Rotation="0,0,10", Scale="1,1,1") {rotationMode=BMM_Additive, alphaBoolBlend {blendOption=Linear}, boneToModify {boneName=spine_02}, rotation {roll=10}}
   n6 = "Transform (Modify) Bone - Bone: spine_01"(ComponentPose=@n5, Alpha=@n3, Translation="0,0,0", Rotation="10,0,10", Scale="1,1,1") {rotationMode=BMM_Additive, alphaBoolBlend {blendOption=Linear}, boneToModify {boneName=spine_01}, rotation {pitch=10, roll=10}}
   n7 = "Transform (Modify) Bone - Bone: head"(ComponentPose=@n6, Alpha=@n3, Translation="0,0,0", Rotation="0,0,-22", Scale="1,1,1") {rotationMode=BMM_Additive, alphaBoolBlend {blendOption=Linear}, boneToModify {boneName=head}, rotation {roll=-22}}
   n8 = "Transform (Modify) Bone - Bone: neck_01"(ComponentPose=@n7, Alpha=@n3, Translation="0,0,0", Rotation="-5.113,21.316,-9.715", Scale="1,1,1") {rotationMode=BMM_Additive, alphaBoolBlend {blendOption=Linear}, boneToModify {boneName=neck_01}, rotation=<rot -5.113 21.316 -9.715>}
   Output Pose(Result=Component To Local(ComponentPose=@n8))
  ]}
 EventGraph {nodes [
   Event BlueprintUpdateAnimation()
   > Branch(Condition=Get bIsInFrontEnd())
   > Set Frontend_Backpack_Pose(Frontend_Backpack_Pose:float=GetCurveValue(CurveName=Frontend_Backpack_Pose)) -> (Output_Get:float)
   TryGetPawnOwner()
  ]}
]
variables [
 FrontEnd Pickaxe Pose {type=number}
 Frontend_Backpack_Pose {type=number}
]
